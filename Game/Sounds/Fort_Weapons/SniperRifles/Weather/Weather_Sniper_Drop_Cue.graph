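# Sound cue for dropping the Weather sniper rifle, classed under Fort_Loot_Collect.
# A single wave player plays Weather_Sniper_Drop_01 at reduced volume, with no randomization or mixing.
# It uses the shared weapon drop attenuation and the Weapon_Drop_StopOldest concurrency set to limit overlapping drop sounds.
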SoundCue /Game/Sounds/Fort_Weapons/SniperRifles/Weather/Weather_Sniper_Drop_Cue
root=WavePlayer_1
nodes [{id=WavePlayer_1, properties {soundWaveAssetPtr=&/Weather_Sniper_Drop_01}}]
properties {
 attenuationSettings=&Fort_Weapon_Drop_Att
 maxDistance=3000
 soundClassObject=&Fort_Loot_Collect
 volumeMultiplier=0.5
 concurrencyOverrides {resolutionRule=PreventNew}
 concurrencySet [&Weapon_Drop_StopOldest]
}
