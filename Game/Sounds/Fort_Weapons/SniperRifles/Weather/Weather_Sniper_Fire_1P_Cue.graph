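# First-person fire sound cue for the Weather sniper rifle, layering the gunshot with delayed mechanical sounds.
# A mixer combines a close shot, which draws on fewer random variants at the last two quality levels, with quieter, later layers.
# A pitched-down pistol clip eject follows about half a second after the shot, then a faint MilitarySniper casing just over a second after it.
# It uses first-person weapon attenuation and gunfire concurrency.
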SoundCue /Game/Sounds/Fort_Weapons/SniperRifles/Weather/Weather_Sniper_Fire_1P_Cue
root=Modulator_17
nodes [
 {id=Delay_0, children [Random_3], properties {delayMax=1.1, delayMin=1.1}}
 {id=Delay_1, children [Random_4], properties {delayMax=0.55, delayMin=0.55}}
 {id=Mixer_0, children [QualityLevel_1, Modulator_0, Delay_0], properties {inputVolume [1,0.6,0.2]}}
 {id=Modulator_0, children [Delay_1], properties {pitchMax=0.95, pitchMin=0.9}}
 {id=Modulator_17, children [Mixer_0]}
 {id=QualityLevel_1, children [Random_0,Random_0,Random_2,Random_2]}
 {id=Random_0, children [WavePlayer_0,WavePlayer_2,WavePlayer_3,WavePlayer_4,WavePlayer_5], properties {weights [1,1,1,1,1]}}
 {id=Random_2, children [WavePlayer_6,WavePlayer_7], properties {weights [1,1]}}
 {id=Random_3, children [WavePlayer_13, WavePlayer_14, WavePlayer_15], properties {weights [1,1,1]}}
 {id=Random_4, children [WavePlayer_17, WavePlayer_18], properties {weights [1,1]}}
 {id=WavePlayer_0, properties {soundWaveAssetPtr=&/Weather_Sniper_Fire_Close_1P_01}}
 {id=WavePlayer_13, properties {soundWaveAssetPtr=/Game/Sounds/Fort_Weapons/SniperRifles/MilitarySniper/MilitarySniper_Casing_01}}
 {id=WavePlayer_14, properties {soundWaveAssetPtr=/Game/Sounds/Fort_Weapons/SniperRifles/MilitarySniper/MilitarySniper_Casing_02}}
 {id=WavePlayer_15, properties {soundWaveAssetPtr=/Game/Sounds/Fort_Weapons/SniperRifles/MilitarySniper/MilitarySniper_Casing_03}}
 {id=WavePlayer_17, properties {soundWaveAssetPtr=/Game/Sounds/Fort_Weapons/pistol/pistol_auto_eject_clip_01}}
 {id=WavePlayer_18, properties {soundWaveAssetPtr=/Game/Sounds/Fort_Weapons/pistol/pistol_auto_eject_clip_02}}
 {id=WavePlayer_2, properties {soundWaveAssetPtr=&/Weather_Sniper_Fire_Close_1P_02}}
 {id=WavePlayer_3, properties {soundWaveAssetPtr=&/Weather_Sniper_Fire_Close_1P_03}}
 {id=WavePlayer_4, properties {soundWaveAssetPtr=&/Weather_Sniper_Fire_Close_1P_04}}
 {id=WavePlayer_5, properties {soundWaveAssetPtr=&/Weather_Sniper_Fire_Close_1P_05}}
 {id=WavePlayer_6, properties {soundWaveAssetPtr=&Weather_Sniper_Fire_Close_1P_01}}
 {id=WavePlayer_7, properties {soundWaveAssetPtr=&Weather_Sniper_Fire_Close_1P_02}}
]
properties {
 attenuationSettings=&Fort_Weapon_Projectile_FirstPerson_Athena_Att
 maxDistance=2097152
 soundClassObject=&Fort_WeaponRanged_Fire_1P_PSM
 volumeMultiplier=0.8
 attenuationOverrides {distanceAlgorithm=NaturalSound, falloffDistance=4000, attenuationShapeExtents {x=1000}}
 concurrencyOverrides {bLimitToOwner=true, maxCount=3, resolutionRule=StopOldest}
 concurrencySet [&Weapon_Gunfire_FirstPerson_Conc]
}
