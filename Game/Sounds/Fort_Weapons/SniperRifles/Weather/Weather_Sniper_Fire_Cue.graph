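# Weather sniper rifle fire SoundCue, mixing close, medium and distant gunshot layers with separate attenuation and sound classes (third-person fire for close, distant fire for the others).
# Each layer randomizes three variants on two quality levels and plays a single fixed wave on the other two.
# The randomized close branch also mixes in a delayed, pitched-down pistol clip-eject sound.
# Local and global gunfire concurrency sets limit simultaneous playback.
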SoundCue /Game/Sounds/Fort_Weapons/SniperRifles/Weather/Weather_Sniper_Fire_Cue
root=Mixer_1
nodes [
 {
  id=Attenuation_11
  children [Modulator_17]
  properties {attenuationSettings=&Fort_Weapon_Projectile_Close_Athena_Att}
 }
 {
  id=Attenuation_12
  children [Modulator_18]
  properties {attenuationSettings=&Fort_Weapon_Projectile_Medium_Mono_Athena_Att}
 }
 {
  id=Attenuation_13
  children [Modulator_19]
  properties {attenuationSettings=&Fort_Weapon_Projectile_Distant_Mono_Athena_Att}
 }
 {id=Delay_1, children [Random_4], properties {delayMax=0.55, delayMin=0.55}}
 {id=Mixer_0, children [Random_1,Modulator_0], properties {inputVolume [1,0.6]}}
 {id=Mixer_1, children [SoundClass_14, SoundClass_16, SoundClass_15], properties {inputVolume [0.9,0.6,0.3]}}
 {id=Modulator_0, children [Delay_1], properties {pitchMax=0.95, pitchMin=0.9}}
 {id=Modulator_17, children [QualityLevel_1]}
 {id=Modulator_18, children [QualityLevel_2], properties {volumeMax=2.2, volumeMin=2.1}}
 {id=Modulator_19, children [QualityLevel_0], properties {volumeMax=2.05, volumeMin=1.95}}
 {id=QualityLevel_0, children [Random_3, Random_3, WavePlayer_19, WavePlayer_19]}
 {id=QualityLevel_1, children [Mixer_0,Mixer_0,WavePlayer_8,WavePlayer_8]}
 {id=QualityLevel_2, children [Random_2, Random_2, WavePlayer_12, WavePlayer_12]}
 {id=Random_1, children [WavePlayer_5,WavePlayer_6,WavePlayer_7], properties {weights [1,1,1]}}
 {id=Random_2, children [WavePlayer_9, WavePlayer_10, WavePlayer_11], properties {weights [1,1,1]}}
 {id=Random_3, children [WavePlayer_16, WavePlayer_17, WavePlayer_18], properties {weights [1,1,1]}}
 {id=Random_4, children [WavePlayer_0,WavePlayer_1], properties {weights [1,1]}}
 {
  id=SoundClass_14
  children [Attenuation_11]
  properties {soundClassOverride=&Fort_WeaponRanged_Fire_3P_PSM}
 }
 {
  id=SoundClass_15
  children [Attenuation_13]
  properties {soundClassOverride=&Fort_WeaponRanged_Fire_Distant}
 }
 {
  id=SoundClass_16
  children [Attenuation_12]
  properties {soundClassOverride=&Fort_WeaponRanged_Fire_Distant}
 }
 {id=WavePlayer_0, properties {soundWaveAssetPtr=/Game/Sounds/Fort_Weapons/pistol/pistol_auto_eject_clip_01}}
 {id=WavePlayer_1, properties {soundWaveAssetPtr=/Game/Sounds/Fort_Weapons/pistol/pistol_auto_eject_clip_02}}
 {id=WavePlayer_10, properties {soundWaveAssetPtr=&/Weather_Sniper_Fire_Medium_02}}
 {id=WavePlayer_11, properties {soundWaveAssetPtr=&/Weather_Sniper_Fire_Medium_03}}
 {id=WavePlayer_12, properties {soundWaveAssetPtr=&/Weather_Sniper_Fire_Medium_01}}
 {id=WavePlayer_16, properties {soundWaveAssetPtr=&/Weather_Sniper_Fire_Distant_01}}
 {id=WavePlayer_17, properties {soundWaveAssetPtr=&/Weather_Sniper_Fire_Distant_02}}
 {id=WavePlayer_18, properties {soundWaveAssetPtr=&/Weather_Sniper_Fire_Distant_03}}
 {id=WavePlayer_19, properties {soundWaveAssetPtr=&Weather_Sniper_Fire_Distant_01}}
 {id=WavePlayer_5, properties {soundWaveAssetPtr=&/Weather_Sniper_Fire_Close_01}}
 {id=WavePlayer_6, properties {soundWaveAssetPtr=&/Weather_Sniper_Fire_Close_02}}
 {id=WavePlayer_7, properties {soundWaveAssetPtr=&/Weather_Sniper_Fire_Close_03}}
 {id=WavePlayer_8, properties {soundWaveAssetPtr=&Weather_Sniper_Fire_Close_01}}
 {id=WavePlayer_9, properties {soundWaveAssetPtr=&Weather_Sniper_Fire_Medium_01}}
]
properties {
 maxDistance=27500
 soundClassObject=&Fort_WeaponRanged_Fire_3P_PSM
 volumeMultiplier=1
 attenuationOverrides {distanceAlgorithm=NaturalSound, falloffDistance=4000, attenuationShapeExtents {x=1000}}
 concurrencyOverrides {bLimitToOwner=true, maxCount=3, resolutionRule=StopOldest}
 concurrencySet [&Weapon_Gunfire_Local_Conc, &Weapon_Gunfire_Global_Conc]
}
