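# Looping sound cue for the Weather sniper rifle's scope scan, played in the weapon foley sound class.
# A mixer blends three looping layers at separate volumes: a detect loop, a ping that starts after a short fixed delay, and a scope loop borrowed from the SteamPunk sniper.
# The cue carries its own attenuation and a concurrency override that prevents new instances once the limit is reached.
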SoundCue /Game/Sounds/Fort_Weapons/SniperRifles/Weather/Weather_Sniper_Scope_ScanLoop_Cue
root=Mixer_0
nodes [
 {id=Delay_0, children [WavePlayer_1], properties {delayMax=0.5, delayMin=0.5}}
 {id=Mixer_0, children [WavePlayer_0,Delay_0,WavePlayer_2], properties {inputVolume [0.7,0.5,1]}}
 {
  id=WavePlayer_0
  properties {bLooping=true, soundWaveAssetPtr=&/Weather_Sniper_Scope_DetectLoop_01}
 }
 {id=WavePlayer_1, properties {bLooping=true, soundWaveAssetPtr=&/Weather_Sniper_Scope_Ping_01}}
 {id=WavePlayer_2, properties {bLooping=true, soundWaveAssetPtr=/Game/Sounds/Fort_Weapons/SteamPunk/Sniper/SP_Sniper_Scope_Loop_01}}
]
properties {
 bOverrideAttenuation=true
 bOverrideConcurrency=true
 maxDistance=2700
 soundClassObject=&Fort_WeaponFoley
 volumeMultiplier=0.351
 attenuationOverrides {falloffDistance=1500, nonSpatializedRadiusStart=200, attenuationShapeExtents {x=1200}}
 concurrencyOverrides {resolutionRule=PreventNew}
}
