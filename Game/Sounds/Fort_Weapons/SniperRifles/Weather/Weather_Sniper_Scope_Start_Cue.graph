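# Sound cue for the Weather sniper rifle's scope-start sound, classed as weapon foley.
# A single wave player feeds the output directly, with no randomization or layering, at lowered volume and pitch.
# The cue overrides attenuation for a limited audible range and overrides concurrency to prevent new instances once the limit is reached.
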SoundCue /Game/Sounds/Fort_Weapons/SniperRifles/Weather/Weather_Sniper_Scope_Start_Cue
root=WavePlayer_3
nodes [{id=WavePlayer_3, properties {soundWaveAssetPtr=&/Weather_Sniper_Scope_Start_01}}]
properties {
 bOverrideAttenuation=true
 bOverrideConcurrency=true
 maxDistance=2700
 pitchMultiplier=0.8
 soundClassObject=&Fort_WeaponFoley
 volumeMultiplier=0.2
 attenuationOverrides {falloffDistance=1500, nonSpatializedRadiusStart=200, attenuationShapeExtents {x=1200}}
 concurrencyOverrides {resolutionRule=PreventNew}
}
